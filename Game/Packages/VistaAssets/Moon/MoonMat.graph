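# Unlit translucent material for a moon, kept among the vista assets and drawn before depth of field.
# It multiplies the RGB of the Moon texture from the Fortress_Sky package by a Color vector parameter.
# The Color default is an overbright neutral value, so the moon's brightness and tint can be adjusted through that one parameter.
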Material /Game/Packages/VistaAssets/Moon/MoonMat
graph {
 properties {
  blendMode=BLEND_Translucent
  pixelDepthOffsetMode=PDOM_Legacy
  shadingModel=MSM_Unlit
  translucencyPass=MTP_BeforeDOF
  usedShadingModels=Unlit
 }
 nodes [Multiply(
   A=TextureSample() {texture=/Game/Packages/Fortress_Sky/Textures/Moon}.RGB
   B=VectorParameter() {parameterName=Color, defaultValue=<rgba 4 4 4 0>}.RGB
  )]
}
