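# Unlit additive material for a moon glow, drawn as a soft radial gradient.
# A centered UV sphere mask is tinted by the color parameter, then raised to the falloff power packed in that parameter's alpha.
# A world position offset chain scales vertices about the component pivot and feeds the scaled position into a clamped CameraOffset, shifting the glow along the camera axis by a very large parameterized distance.
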Material /Game/Packages/VistaAssets/Moon/moonGlow
graph {
 properties {blendMode=BLEND_Additive, pixelDepthOffsetMode=PDOM_Legacy, shadingModel=MSM_Unlit, usedShadingModels=Unlit}
 nodes [
  n1 = WorldPosition()
  n2 = Multiply(
   A=Subtract(
    A=@n1.XYZ
    B=MaterialFunctionCall() {materialFunction=&ComponentPivotLocation}.Result
   )
   B=ScalarParameter() {defaultValue=4.3468, parameterName=Scale}
  )
  Add(
   A=MaterialFunctionCall(Offset Amount=ScalarParameter() {defaultValue=-3390866.75, parameterName=OffsetAmount}, World Position=Add(A=@n1.XYZ, B=@n2)) {materialFunction=/Engine/Functions/Engine_MaterialFunctions02/WorldPositionOffset/CameraOffset}.Clamped Camera Offset
   B=@n2
  )
  n3 = TextureCoordinate()
  n4 = Constant() {r=0.5}
  n5 = VectorParameter() {parameterName=Color And Falloff Power, defaultValue=<rgba 0.4398 0.465 0.485 3.3381>}
  Power(Base=Multiply(A=SphereMask(A=@n3, B=@n4, Radius=@n4) {hardnessPercent=0}, B=@n5.RGB), Exp=@n5.A)
 ]
}
